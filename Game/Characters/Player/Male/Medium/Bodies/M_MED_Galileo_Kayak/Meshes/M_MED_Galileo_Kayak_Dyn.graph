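# Physics asset for the dynamic skirt on the medium male Galileo Kayak player body.
# Box and capsule skirt chains hang from the pelvis, the front and side ones via small MSTR bodies, with pelvis, spine_01 and thigh capsules as body shapes.
# Chain joints limit swing and tightly limit twist, especially narrowing one swing axis on front-mid chains.
# Free-rotation cross-constraints tie neighboring front and side chains together; back chains stay unlinked.
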PhysicsAsset /Game/Characters/Player/Male/Medium/Bodies/M_MED_Galileo_Kayak/Meshes/M_MED_Galileo_Kayak_Dyn
bodies {
 dyn_skirt_bk_01 [{shapeType=Box, extentX=11.1875, extentY=11.3125, extentZ=2.1475, center=<5.2189 0 0>, rotation=<rot -2.8568 0 0>}]
 dyn_skirt_bk_02 [{shapeType=Capsule, length=27.3913, radius=6.0952, shapeName=dyn_skirt_bk_l_02_capsule, center=<5.4063 0 3>, rotation=<rot 0 0 90>}]
 dyn_skirt_bk_03 [{shapeType=Capsule, length=29.3387, radius=7.6155, shapeName=dyn_skirt_bk_l_02_capsule, center=<5.4063 0 3>, rotation=<rot 0 0 90>}]
 dyn_skirt_bk_l_01 [{shapeType=Box, extentX=11.1875, extentY=11.3125, extentZ=2.1475, center=<5.2189 0 0>, rotation=<rot -2.8568 0 0>}]
 dyn_skirt_bk_l_02 [{shapeType=Capsule, length=21.9209, radius=5, shapeName=dyn_skirt_bk_l_02_capsule, center=<5.4063 0 3>, rotation=<rot 0 0 90>}]
 dyn_skirt_bk_l_03 [{shapeType=Capsule, length=27.7878, radius=6.4895, shapeName=dyn_skirt_bk_l_02_capsule, center=<5.4063 0 3>, rotation=<rot 0 0 90>}]
 dyn_skirt_bk_r_01 [{shapeType=Box, extentX=11.1875, extentY=11.3125, extentZ=2.1475, center=<5.2189 0 0>, rotation=<rot -2.8568 0 0>}]
 dyn_skirt_bk_r_02 [{shapeType=Capsule, length=20.8038, radius=6.1172, shapeName=dyn_skirt_bk_l_02_capsule, center=<5.4063 0 3>, rotation=<rot 0 0 90>}]
 dyn_skirt_bk_r_03 [{shapeType=Capsule, length=26.4614, radius=7.8159, shapeName=dyn_skirt_bk_l_02_capsule, center=<5.4063 0 3>, rotation=<rot -0.0001 0 87.1137>}]
 dyn_skirt_fr_MSTR [{shapeType=Capsule, length=1, radius=1}]
 dyn_skirt_fr_l_01 [{shapeType=Capsule, length=10.6229, radius=6.4671, shapeName=dyn_skirt_r_03_capsule, center=<5.2553 0 4>, rotation=<rot 0 0 90>}]
 dyn_skirt_fr_l_02 [{shapeType=Capsule, length=28.3009, radius=6.4851, shapeName=dyn_skirt_r_03_capsule, center=<5.2553 0 4>, rotation=<rot 0 0 90>}]
 dyn_skirt_fr_l_03 [{shapeType=Capsule, length=27.8519, radius=6.9341, shapeName=dyn_skirt_r_03_capsule, center=<5.2553 0 4>, rotation=<rot 0 0 90>}]
 dyn_skirt_fr_mid_l_01 [{shapeType=Capsule, length=10.6229, radius=6.4671, shapeName=dyn_skirt_r_03_capsule, center=<5.2553 0 4>, rotation=<rot 0 0 90>}]
 dyn_skirt_fr_mid_l_02 [{shapeType=Capsule, length=25.816, radius=6.4851, shapeName=dyn_skirt_r_03_capsule, center=<5.2553 0 4>, rotation=<rot 0 0 90>}]
 dyn_skirt_fr_mid_l_03 [{shapeType=Capsule, length=25.367, radius=6.9341, shapeName=dyn_skirt_r_03_capsule, center=<5.2553 0 4>, rotation=<rot 0 0 90>}]
 dyn_skirt_fr_mid_l_MSTR [{shapeType=Capsule, length=1, radius=1}]
 dyn_skirt_fr_mid_r_01 [{shapeType=Capsule, length=10.6229, radius=6.4671, shapeName=dyn_skirt_r_03_capsule, center=<5.2553 0 4>, rotation=<rot 0 0 90>}]
 dyn_skirt_fr_mid_r_02 [{shapeType=Capsule, length=25.816, radius=6.4851, shapeName=dyn_skirt_r_03_capsule, center=<5.2553 0 4>, rotation=<rot 0.0002 0 96.3416>}]
 dyn_skirt_fr_mid_r_03 [{
   shapeType=Capsule
   length=25.367
   radius=6.9341
   shapeName=dyn_skirt_r_03_capsule
   center=<5.2553 0 4>
   rotation=<rot 0.0002 -0.0001 92.8448>
  }]
 dyn_skirt_fr_mid_r_MSTR [{shapeType=Capsule, length=1, radius=1}]
 dyn_skirt_fr_r_01 [{shapeType=Capsule, length=10.6229, radius=6.4671, shapeName=dyn_skirt_r_03_capsule, center=<5.2553 0 4>, rotation=<rot 0 0 90>}]
 dyn_skirt_fr_r_02 [{shapeType=Capsule, length=28.3009, radius=6.4851, shapeName=dyn_skirt_r_03_capsule, center=<5.2553 0 4>, rotation=<rot 0 0 90>}]
 dyn_skirt_fr_r_03 [{shapeType=Capsule, length=27.8519, radius=6.9341, shapeName=dyn_skirt_r_03_capsule, center=<5.2553 0 4>, rotation=<rot 0 0 90>}]
 dyn_skirt_l_01 [{shapeType=Box, extentX=11.1875, extentY=11.3125, extentZ=1.3125, center=<5.2189 0 0>}]
 dyn_skirt_l_02 [{shapeType=Capsule, length=25.5146, radius=6.7865, shapeName=dyn_skirt_r_02_capsule, center=<5.2553 0 4>, rotation=<rot 0 0 90>}]
 dyn_skirt_l_03 [{shapeType=Capsule, length=25.5146, radius=6.7865, shapeName=dyn_skirt_r_02_capsule, center=<5.2553 0 4>, rotation=<rot 0 0 90>}]
 dyn_skirt_l_MSTR [{shapeType=Capsule, length=1, radius=1}]
 dyn_skirt_r_01 [{shapeType=Box, extentX=11.1875, extentY=11.3125, extentZ=1.3125, center=<5.2189 0 0>}]
 dyn_skirt_r_02 [{shapeType=Capsule, length=25.5146, radius=6.7865, shapeName=dyn_skirt_r_02_capsule, center=<5.2553 0 4>, rotation=<rot 0 0 90>}]
 dyn_skirt_r_03 [{shapeType=Capsule, length=25.5146, radius=6.7865, shapeName=dyn_skirt_r_03_capsule, center=<5.2553 0 4>, rotation=<rot 0 -0.0001 86.0801>}]
 dyn_skirt_r_MSTR [{shapeType=Capsule, length=1, radius=1}]
 pelvis [{shapeType=Capsule, length=8.125, radius=11.5312, center=<-8.3294 0 -0.0007>}]
 spine_01 [{shapeType=Capsule, length=9.5312, radius=10, center=<-7.086 -0.0634 -0.0007>}]
 thigh_l [{shapeType=Capsule, length=24.8125, radius=11.5625, center=<-21.3813 -0.7187 2.3763>, rotation=<rot 86.5422 179.6205 178.886>}]
 thigh_r [{shapeType=Capsule, length=31.3438, radius=10.1562, center=<22.4713 -0.116 -0.3555>, rotation=<rot 87.6218 176.9848 172.4833>}]
}
constraints [
 {bone1Name=dyn_skirt_bk_l_02, bone2Name=dyn_skirt_bk_l_01, swing1LimitDegrees=45, swing2LimitDegrees=45, twistLimitDegrees=5}
 {bone1Name=dyn_skirt_bk_l_03, bone2Name=dyn_skirt_bk_l_02, swing1LimitDegrees=45, swing2LimitDegrees=45, twistLimitDegrees=5}
 {bone1Name=dyn_skirt_bk_02, bone2Name=dyn_skirt_bk_01, swing1LimitDegrees=45, swing2LimitDegrees=45, twistLimitDegrees=5}
 {bone1Name=dyn_skirt_bk_03, bone2Name=dyn_skirt_bk_02, swing1LimitDegrees=45, swing2LimitDegrees=45, twistLimitDegrees=5}
 {bone1Name=dyn_skirt_fr_mid_l_01, bone2Name=dyn_skirt_fr_mid_l_MSTR, swing1LimitDegrees=2, swing2LimitDegrees=45, twistLimitDegrees=5}
 {bone1Name=dyn_skirt_fr_mid_l_02, bone2Name=dyn_skirt_fr_mid_l_01, swing1LimitDegrees=2, swing2LimitDegrees=45, twistLimitDegrees=5}
 {bone1Name=dyn_skirt_fr_mid_l_03, bone2Name=dyn_skirt_fr_mid_l_02, swing1LimitDegrees=2, swing2LimitDegrees=45, twistLimitDegrees=5}
 {bone1Name=dyn_skirt_fr_l_01, bone2Name=dyn_skirt_fr_MSTR, swing1LimitDegrees=45, swing2LimitDegrees=45, twistLimitDegrees=5}
 {bone1Name=dyn_skirt_fr_l_02, bone2Name=dyn_skirt_fr_l_01, swing1LimitDegrees=20, swing2LimitDegrees=20, twistLimitDegrees=5}
 {bone1Name=dyn_skirt_fr_l_03, bone2Name=dyn_skirt_fr_l_02, swing1LimitDegrees=20, swing2LimitDegrees=20, twistLimitDegrees=5}
 {bone1Name=dyn_skirt_fr_r_01, bone2Name=dyn_skirt_fr_MSTR, swing1LimitDegrees=45, swing2LimitDegrees=45, twistLimitDegrees=5}
 {bone1Name=dyn_skirt_fr_r_02, bone2Name=dyn_skirt_fr_r_01, swing1LimitDegrees=20, swing2LimitDegrees=20, twistLimitDegrees=5}
 {bone1Name=dyn_skirt_fr_r_03, bone2Name=dyn_skirt_fr_r_02, swing1LimitDegrees=20, swing2LimitDegrees=20, twistLimitDegrees=5}
 {bone1Name=dyn_skirt_l_01, bone2Name=dyn_skirt_l_MSTR, swing1LimitDegrees=45, swing2LimitDegrees=45, twistLimitDegrees=5}
 {bone1Name=dyn_skirt_l_02, bone2Name=dyn_skirt_l_01, swing1LimitDegrees=45, swing2LimitDegrees=45, twistLimitDegrees=5}
 {bone1Name=dyn_skirt_l_03, bone2Name=dyn_skirt_l_02, swing1LimitDegrees=45, swing2LimitDegrees=45, twistLimitDegrees=5}
 {bone1Name=dyn_skirt_fr_mid_r_01, bone2Name=dyn_skirt_fr_mid_r_MSTR, swing1LimitDegrees=3, swing2LimitDegrees=45, twistLimitDegrees=5}
 {bone1Name=dyn_skirt_fr_mid_r_02, bone2Name=dyn_skirt_fr_mid_r_01, swing1LimitDegrees=3, swing2LimitDegrees=45, twistLimitDegrees=5}
 {bone1Name=dyn_skirt_fr_mid_r_03, bone2Name=dyn_skirt_fr_mid_r_02, swing1LimitDegrees=3, swing2LimitDegrees=45, twistLimitDegrees=5}
 {bone1Name=dyn_skirt_r_01, bone2Name=dyn_skirt_r_MSTR, swing1LimitDegrees=45, swing2LimitDegrees=45, twistLimitDegrees=5}
 {bone1Name=dyn_skirt_r_02, bone2Name=dyn_skirt_r_01, swing1LimitDegrees=45, swing2LimitDegrees=45, twistLimitDegrees=5}
 {bone1Name=dyn_skirt_r_03, bone2Name=dyn_skirt_r_02, swing1LimitDegrees=45, swing2LimitDegrees=45, twistLimitDegrees=5}
 {bone1Name=dyn_skirt_bk_r_02, bone2Name=dyn_skirt_bk_r_01, swing1LimitDegrees=45, swing2LimitDegrees=45, twistLimitDegrees=5}
 {bone1Name=dyn_skirt_bk_r_03, bone2Name=dyn_skirt_bk_r_02, swing1LimitDegrees=45, swing2LimitDegrees=45, twistLimitDegrees=5}
 {bone1Name=dyn_skirt_bk_l_01, bone2Name=pelvis, swing1LimitDegrees=45, swing2LimitDegrees=45, twistLimitDegrees=5}
 {bone1Name=dyn_skirt_bk_01, bone2Name=pelvis, swing1LimitDegrees=45, swing2LimitDegrees=45, twistLimitDegrees=5}
 {
  bone1Name=dyn_skirt_fr_mid_l_MSTR
  bone2Name=pelvis
  swing1LimitDegrees=45
  swing1Motion=Free
  swing2LimitDegrees=45
  swing2Motion=Free
  twistLimitDegrees=45
  twistMotion=Free
 }
 {
  bone1Name=dyn_skirt_fr_MSTR
  bone2Name=pelvis
  swing1LimitDegrees=45
  swing1Motion=Free
  swing2LimitDegrees=45
  swing2Motion=Free
  twistLimitDegrees=45
  twistMotion=Free
 }
 {
  bone1Name=dyn_skirt_l_MSTR
  bone2Name=pelvis
  swing1LimitDegrees=45
  swing1Motion=Free
  swing2LimitDegrees=45
  swing2Motion=Free
  twistLimitDegrees=45
  twistMotion=Free
 }
 {
  bone1Name=dyn_skirt_fr_mid_r_MSTR
  bone2Name=pelvis
  swing1LimitDegrees=45
  swing1Motion=Free
  swing2LimitDegrees=45
  swing2Motion=Free
  twistLimitDegrees=45
  twistMotion=Free
 }
 {
  bone1Name=dyn_skirt_r_MSTR
  bone2Name=pelvis
  swing1LimitDegrees=45
  swing1Motion=Free
  swing2LimitDegrees=45
  swing2Motion=Free
  twistLimitDegrees=45
  twistMotion=Free
 }
 {bone1Name=dyn_skirt_bk_r_01, bone2Name=pelvis, swing1LimitDegrees=45, swing2LimitDegrees=45, twistLimitDegrees=5}
 {
  bone1Name=thigh_l
  bone2Name=pelvis
  swing1LimitDegrees=45
  swing1Motion=Free
  swing2LimitDegrees=45
  swing2Motion=Free
  twistLimitDegrees=45
  twistMotion=Free
 }
 {
  bone1Name=thigh_r
  bone2Name=pelvis
  swing1LimitDegrees=45
  swing1Motion=Free
  swing2LimitDegrees=45
  swing2Motion=Free
  twistLimitDegrees=45
  twistMotion=Free
 }
 {
  bone1Name=spine_01
  bone2Name=pelvis
  swing1LimitDegrees=45
  swing1Motion=Free
  swing2LimitDegrees=45
  swing2Motion=Free
  twistLimitDegrees=45
  twistMotion=Free
 }
 {
  bone1Name=dyn_skirt_fr_l_03
  bone2Name=dyn_skirt_fr_mid_l_03
  swing1LimitDegrees=45
  swing1Motion=Free
  swing2LimitDegrees=45
  swing2Motion=Free
  twistLimitDegrees=45
  twistMotion=Free
 }
 {
  bone1Name=dyn_skirt_fr_r_03
  bone2Name=dyn_skirt_fr_mid_r_03
  swing1LimitDegrees=45
  swing1Motion=Free
  swing2LimitDegrees=45
  swing2Motion=Free
  twistLimitDegrees=45
  twistMotion=Free
 }
 {
  bone1Name=dyn_skirt_l_03
  bone2Name=dyn_skirt_fr_mid_l_03
  swing1LimitDegrees=45
  swing1Motion=Free
  swing2LimitDegrees=45
  swing2Motion=Free
  twistLimitDegrees=45
  twistMotion=Free
 }
 {
  bone1Name=dyn_skirt_r_03
  bone2Name=dyn_skirt_fr_mid_r_03
  swing1LimitDegrees=45
  swing1Motion=Free
  swing2LimitDegrees=45
  swing2Motion=Free
  twistLimitDegrees=45
  twistMotion=Free
 }
 {
  bone1Name=dyn_skirt_fr_mid_l_02
  bone2Name=dyn_skirt_l_02
  swing1LimitDegrees=45
  swing1Motion=Free
  swing2LimitDegrees=45
  swing2Motion=Free
  twistLimitDegrees=45
  twistMotion=Free
 }
 {
  bone1Name=dyn_skirt_fr_mid_r_02
  bone2Name=dyn_skirt_r_02
  swing1LimitDegrees=45
  swing1Motion=Free
  swing2LimitDegrees=45
  swing2Motion=Free
  twistLimitDegrees=45
  twistMotion=Free
 }
]
