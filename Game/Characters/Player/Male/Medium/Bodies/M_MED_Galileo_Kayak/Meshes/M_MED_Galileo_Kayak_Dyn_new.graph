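# Physics asset for the dynamic skirt on the medium male Galileo Kayak player body, with capsules on every skirt bone plus the pelvis, lower spine and thighs.
# Three-bone chains with limited swing and twist hang from the pelvis, directly at the back and through free-rotating MSTR bones at front and sides.
# Extra free-rotation constraints link the last bone of each side chain to its neighboring front-mid chain, tying those panels together.
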PhysicsAsset /Game/Characters/Player/Male/Medium/Bodies/M_MED_Galileo_Kayak/Meshes/M_MED_Galileo_Kayak_Dyn_new
bodies {
 dyn_skirt_bk_01 [{shapeType=Capsule, length=11, radius=3.75, shapeName=dyn_skirt_bk_01_capsule, center=<3.5 0 2.5>, rotation=<rot 0 0 90>}]
 dyn_skirt_bk_02 [{shapeType=Capsule, length=10.5, radius=4.5, shapeName=dyn_skirt_bk_02_capsule, center=<2.1701 0 2>, rotation=<rot 0 0 90>}]
 dyn_skirt_bk_03 [{shapeType=Capsule, length=9, radius=6, shapeName=dyn_skirt_bk_03_capsule, center=<4.4917 0 0.75>, rotation=<rot 0 0 90>}]
 dyn_skirt_bk_l_01 [{shapeType=Capsule, length=11, radius=3.75, shapeName=dyn_skirt_bk_l_01_capsule, center=<3.5 0 2.5>, rotation=<rot 0 0 90>}]
 dyn_skirt_bk_l_02 [{shapeType=Capsule, length=10.5, radius=4.5, shapeName=dyn_skirt_bk_l_02_capsule, center=<2.1701 0 2>, rotation=<rot 0 0 90>}]
 dyn_skirt_bk_l_03 [{shapeType=Capsule, length=9, radius=6, shapeName=dyn_skirt_bk_l_03_capsule, center=<4.4917 0 0.75>, rotation=<rot 0 0 90>}]
 dyn_skirt_bk_r_01 [{shapeType=Capsule, length=11, radius=3.75, shapeName=dyn_skirt_bk_r_01_capsule, center=<3.5 0 2.5>, rotation=<rot 0 0 90>}]
 dyn_skirt_bk_r_02 [{shapeType=Capsule, length=10.5, radius=4.5, shapeName=dyn_skirt_bk_r_02_capsule, center=<2.1701 0 2>, rotation=<rot 0 0 90>}]
 dyn_skirt_bk_r_03 [{shapeType=Capsule, length=9, radius=6, shapeName=dyn_skirt_bk_r_03_capsule, center=<4.4917 0 0.75>, rotation=<rot 0 0 90>}]
 dyn_skirt_fr_MSTR [{shapeType=Capsule, length=1, radius=1}]
 dyn_skirt_fr_l_01 [{shapeType=Capsule, length=4, radius=2, shapeName=dyn_skirt_fr_l_01_capsule, rotation=<rot 0 0 90>}]
 dyn_skirt_fr_l_02 [{shapeType=Capsule, length=5, radius=2, shapeName=dyn_skirt_fr_l_02_capsule, rotation=<rot 0 0 90>}]
 dyn_skirt_fr_l_03 [{shapeType=Capsule, length=10, radius=5.5, shapeName=dyn_skirt_fr_mid_r_03_capsule, center=<4 -8 0>, rotation=<rot 0 0 90>}]
 dyn_skirt_fr_mid_l_01 [{shapeType=Capsule, length=11.5, radius=3.5, shapeName=dyn_skirt_fr_mid_l_01_capsule, center=<2.1701 0 1.5>, rotation=<rot 0 0 90>}]
 dyn_skirt_fr_mid_l_02 [{shapeType=Capsule, length=9.5, radius=5.5, shapeName=dyn_skirt_fr_mid_l_02_capsule, center=<2.1701 0 0.5746>, rotation=<rot 0 0 90>}]
 dyn_skirt_fr_mid_l_03 [{shapeType=Capsule, length=10.5, radius=5.5, shapeName=dyn_skirt_fr_mid_l_03_capsule, center=<4 0 0>, rotation=<rot 0 0 90>}]
 dyn_skirt_fr_mid_l_MSTR [{shapeType=Capsule, length=1, radius=1}]
 dyn_skirt_fr_mid_r_01 [{shapeType=Capsule, length=11.5, radius=3.5, shapeName=dyn_skirt_fr_mid_r_01_capsule, center=<2.1701 0 1.5>, rotation=<rot 0 0 90>}]
 dyn_skirt_fr_mid_r_02 [{shapeType=Capsule, length=9.5, radius=5.5, shapeName=dyn_skirt_fr_mid_r_02_capsule, center=<2.1701 0 0.5746>, rotation=<rot 0 0 90>}]
 dyn_skirt_fr_mid_r_03 [{shapeType=Capsule, length=10.5, radius=5.5, shapeName=dyn_skirt_fr_mid_r_03_capsule, center=<4 0 0>, rotation=<rot 0 0 90>}]
 dyn_skirt_fr_mid_r_MSTR [{shapeType=Capsule, length=1, radius=1}]
 dyn_skirt_fr_r_01 [{shapeType=Capsule, length=4, radius=2, shapeName=dyn_skirt_fr_r_01_capsule, rotation=<rot 0 0 90>}]
 dyn_skirt_fr_r_02 [{shapeType=Capsule, length=5, radius=2, shapeName=dyn_skirt_fr_r_02_capsule, rotation=<rot 0 0 90>}]
 dyn_skirt_fr_r_03 [{shapeType=Capsule, length=10, radius=5.5, shapeName=dyn_skirt_fr_mid_r_03_capsule, center=<4 8 0>, rotation=<rot 0 0 90>}]
 dyn_skirt_l_01 [{shapeType=Capsule, length=11, radius=4, shapeName=dyn_skirt_l_01_capsule, center=<2.1701 0 3>, rotation=<rot 0 0 90>}]
 dyn_skirt_l_02 [{shapeType=Capsule, length=10.5, radius=4.5, shapeName=dyn_skirt_l_02_capsule, center=<2.1701 0 3>, rotation=<rot 0 0 90>}]
 dyn_skirt_l_03 [{shapeType=Capsule, length=11.5, radius=5.5, shapeName=dyn_skirt_l_03_capsule, center=<2.1701 0 3>, rotation=<rot 0 0 90>}]
 dyn_skirt_l_MSTR [{shapeType=Capsule, length=1, radius=1}]
 dyn_skirt_r_01 [{shapeType=Capsule, length=11, radius=4, shapeName=dyn_skirt_r_01_capsule, center=<2.1701 0 3>, rotation=<rot 0 0 90>}]
 dyn_skirt_r_02 [{shapeType=Capsule, length=10.5, radius=4.5, shapeName=dyn_skirt_r_02_capsule, center=<2.1701 0 3>, rotation=<rot 0 0 90>}]
 dyn_skirt_r_03 [{shapeType=Capsule, length=11.5, radius=5.5, shapeName=dyn_skirt_r_03_capsule, center=<2.1701 0 3>, rotation=<rot 0 0 90>}]
 dyn_skirt_r_MSTR [{shapeType=Capsule, length=1, radius=1}]
 pelvis [{shapeType=Capsule, length=9, radius=10}]
 spine_01 [{shapeType=Capsule, length=9.5312, radius=5.5, center=<1.11 -0.0634 -0.0007>}]
 thigh_l [{shapeType=Capsule, length=33.5, radius=8.5, center=<-19 0 0.8634>, rotation=<rot 90 0 0>}]
 thigh_r [{shapeType=Capsule, length=33.5, radius=8.5, center=<19 0 -0.7917>, rotation=<rot 90 179.9991 179.9991>}]
}
constraints [
 {bone1Name=dyn_skirt_bk_l_02, bone2Name=dyn_skirt_bk_l_01, swing1LimitDegrees=45, swing2LimitDegrees=45, twistLimitDegrees=5}
 {bone1Name=dyn_skirt_bk_l_03, bone2Name=dyn_skirt_bk_l_02, swing1LimitDegrees=45, swing2LimitDegrees=45, twistLimitDegrees=5}
 {bone1Name=dyn_skirt_bk_02, bone2Name=dyn_skirt_bk_01, swing1LimitDegrees=45, swing2LimitDegrees=45, twistLimitDegrees=5}
 {bone1Name=dyn_skirt_bk_03, bone2Name=dyn_skirt_bk_02, swing1LimitDegrees=45, swing2LimitDegrees=45, twistLimitDegrees=5}
 {bone1Name=dyn_skirt_fr_mid_l_01, bone2Name=dyn_skirt_fr_mid_l_MSTR, swing1LimitDegrees=2, swing2LimitDegrees=45, twistLimitDegrees=5}
 {bone1Name=dyn_skirt_fr_mid_l_02, bone2Name=dyn_skirt_fr_mid_l_01, swing1LimitDegrees=2, swing2LimitDegrees=45, twistLimitDegrees=5}
 {bone1Name=dyn_skirt_fr_mid_l_03, bone2Name=dyn_skirt_fr_mid_l_02, swing1LimitDegrees=2, swing2LimitDegrees=45, twistLimitDegrees=5}
 {bone1Name=dyn_skirt_l_01, bone2Name=dyn_skirt_l_MSTR, swing1LimitDegrees=45, swing2LimitDegrees=45, twistLimitDegrees=5}
 {bone1Name=dyn_skirt_l_02, bone2Name=dyn_skirt_l_01, swing1LimitDegrees=45, swing2LimitDegrees=45, twistLimitDegrees=5}
 {bone1Name=dyn_skirt_l_03, bone2Name=dyn_skirt_l_02, swing1LimitDegrees=45, swing2LimitDegrees=45, twistLimitDegrees=5}
 {bone1Name=dyn_skirt_fr_mid_r_01, bone2Name=dyn_skirt_fr_mid_r_MSTR, swing1LimitDegrees=3, swing2LimitDegrees=45, twistLimitDegrees=5}
 {bone1Name=dyn_skirt_fr_mid_r_02, bone2Name=dyn_skirt_fr_mid_r_01, swing1LimitDegrees=3, swing2LimitDegrees=45, twistLimitDegrees=5}
 {bone1Name=dyn_skirt_fr_mid_r_03, bone2Name=dyn_skirt_fr_mid_r_02, swing1LimitDegrees=3, swing2LimitDegrees=45, twistLimitDegrees=5}
 {bone1Name=dyn_skirt_r_01, bone2Name=dyn_skirt_r_MSTR, swing1LimitDegrees=45, swing2LimitDegrees=45, twistLimitDegrees=5}
 {bone1Name=dyn_skirt_r_02, bone2Name=dyn_skirt_r_01, swing1LimitDegrees=45, swing2LimitDegrees=45, twistLimitDegrees=5}
 {bone1Name=dyn_skirt_r_03, bone2Name=dyn_skirt_r_02, swing1LimitDegrees=45, swing2LimitDegrees=45, twistLimitDegrees=5}
 {bone1Name=dyn_skirt_bk_r_02, bone2Name=dyn_skirt_bk_r_01, swing1LimitDegrees=45, swing2LimitDegrees=45, twistLimitDegrees=5}
 {bone1Name=dyn_skirt_bk_r_03, bone2Name=dyn_skirt_bk_r_02, swing1LimitDegrees=45, swing2LimitDegrees=45, twistLimitDegrees=5}
 {bone1Name=dyn_skirt_bk_l_01, bone2Name=pelvis, swing1LimitDegrees=45, swing2LimitDegrees=45, twistLimitDegrees=5}
 {bone1Name=dyn_skirt_bk_01, bone2Name=pelvis, swing1LimitDegrees=45, swing2LimitDegrees=45, twistLimitDegrees=5}
 {
  bone1Name=dyn_skirt_fr_mid_l_MSTR
  bone2Name=pelvis
  swing1LimitDegrees=45
  swing1Motion=Free
  swing2LimitDegrees=45
  swing2Motion=Free
  twistLimitDegrees=45
  twistMotion=Free
 }
 {
  bone1Name=dyn_skirt_fr_MSTR
  bone2Name=pelvis
  swing1LimitDegrees=45
  swing1Motion=Free
  swing2LimitDegrees=45
  swing2Motion=Free
  twistLimitDegrees=45
  twistMotion=Free
 }
 {
  bone1Name=dyn_skirt_l_MSTR
  bone2Name=pelvis
  swing1LimitDegrees=45
  swing1Motion=Free
  swing2LimitDegrees=45
  swing2Motion=Free
  twistLimitDegrees=45
  twistMotion=Free
 }
 {
  bone1Name=dyn_skirt_fr_mid_r_MSTR
  bone2Name=pelvis
  swing1LimitDegrees=45
  swing1Motion=Free
  swing2LimitDegrees=45
  swing2Motion=Free
  twistLimitDegrees=45
  twistMotion=Free
 }
 {
  bone1Name=dyn_skirt_r_MSTR
  bone2Name=pelvis
  swing1LimitDegrees=45
  swing1Motion=Free
  swing2LimitDegrees=45
  swing2Motion=Free
  twistLimitDegrees=45
  twistMotion=Free
 }
 {bone1Name=dyn_skirt_bk_r_01, bone2Name=pelvis, swing1LimitDegrees=45, swing2LimitDegrees=45, twistLimitDegrees=5}
 {
  bone1Name=thigh_l
  bone2Name=pelvis
  swing1LimitDegrees=45
  swing1Motion=Free
  swing2LimitDegrees=45
  swing2Motion=Free
  twistLimitDegrees=45
  twistMotion=Free
 }
 {
  bone1Name=thigh_r
  bone2Name=pelvis
  swing1LimitDegrees=45
  swing1Motion=Free
  swing2LimitDegrees=45
  swing2Motion=Free
  twistLimitDegrees=45
  twistMotion=Free
 }
 {
  bone1Name=spine_01
  bone2Name=pelvis
  swing1LimitDegrees=45
  swing1Motion=Free
  swing2LimitDegrees=45
  swing2Motion=Free
  twistLimitDegrees=45
  twistMotion=Free
 }
 {
  bone1Name=dyn_skirt_l_03
  bone2Name=dyn_skirt_fr_mid_l_03
  swing1LimitDegrees=45
  swing1Motion=Free
  swing2LimitDegrees=45
  swing2Motion=Free
  twistLimitDegrees=45
  twistMotion=Free
 }
 {
  bone1Name=dyn_skirt_r_03
  bone2Name=dyn_skirt_fr_mid_r_03
  swing1LimitDegrees=45
  swing1Motion=Free
  swing2LimitDegrees=45
  swing2Motion=Free
  twistLimitDegrees=45
  twistMotion=Free
 }
 {bone1Name=dyn_skirt_fr_l_01, bone2Name=dyn_skirt_fr_MSTR, swing1LimitDegrees=10, swing2LimitDegrees=25, twistLimitDegrees=5}
 {bone1Name=dyn_skirt_fr_l_02, bone2Name=dyn_skirt_fr_l_01, swing1LimitDegrees=25, swing2LimitDegrees=25, twistLimitDegrees=5}
 {bone1Name=dyn_skirt_fr_l_03, bone2Name=dyn_skirt_fr_l_02, swing1LimitDegrees=25, swing2LimitDegrees=25, twistLimitDegrees=5}
 {bone1Name=dyn_skirt_fr_r_01, bone2Name=dyn_skirt_fr_MSTR, swing1LimitDegrees=10, swing2LimitDegrees=25, twistLimitDegrees=5}
 {bone1Name=dyn_skirt_fr_r_02, bone2Name=dyn_skirt_fr_r_01, swing1LimitDegrees=25, swing2LimitDegrees=25, twistLimitDegrees=5}
 {bone1Name=dyn_skirt_fr_r_03, bone2Name=dyn_skirt_fr_r_02, swing1LimitDegrees=25, swing2LimitDegrees=25, twistLimitDegrees=5}
]
