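# Physics asset for the dynamic skirt on the Galileo Kayak medium male player body, a skydiving variant by its name.
# Nine three-segment chains of thin boxes form the skirt, alongside capsules for pelvis, lower spine and thighs.
# Back chains hang from the pelvis with tight swing limits, while front and side chains hang from MSTR bodies with looser limits.
# Angularly free cross-constraints tie some neighboring chains together at their lower segments.
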PhysicsAsset /Game/Characters/Player/Male/Medium/Bodies/M_MED_Galileo_Kayak/Meshes/M_MED_Galileo_Kayak_Dyn_skydiving
bodies {
 dyn_skirt_bk_01 [{shapeType=Box, extentX=11.1875, extentY=16.2935, extentZ=1.3125, center=<5.2189 0 0>}]
 dyn_skirt_bk_02 [{shapeType=Box, extentX=11.1875, extentY=20.6235, extentZ=1.3125, center=<5.2189 0 0>}]
 dyn_skirt_bk_03 [{shapeType=Box, extentX=11.1875, extentY=20.6235, extentZ=1.3125, center=<5.2189 0 0>}]
 dyn_skirt_bk_l_01 [{shapeType=Box, extentX=11.1875, extentY=16.2935, extentZ=1.3125, center=<5.2189 0 0>}]
 dyn_skirt_bk_l_02 [{shapeType=Box, extentX=11.1875, extentY=20.6235, extentZ=1.3125, center=<5.2189 0 0>}]
 dyn_skirt_bk_l_03 [{shapeType=Box, extentX=11.1875, extentY=20.6235, extentZ=1.3125, center=<5.2189 0 0>}]
 dyn_skirt_bk_r_01 [{shapeType=Box, extentX=11.1875, extentY=16.2935, extentZ=1.3125, center=<5.2189 0 0>}]
 dyn_skirt_bk_r_02 [{shapeType=Box, extentX=11.1875, extentY=20.6235, extentZ=1.3125, center=<5.2189 0 0>}]
 dyn_skirt_bk_r_03 [{shapeType=Box, extentX=11.1875, extentY=20.6235, extentZ=1.3125, center=<5.2189 0 0>}]
 dyn_skirt_fr_MSTR [{shapeType=Capsule, length=1, radius=1}]
 dyn_skirt_fr_l_01 [{shapeType=Box, extentX=11.1875, extentY=16.2935, extentZ=1.3125, center=<5.2189 -4.2 0>}]
 dyn_skirt_fr_l_02 [{shapeType=Box, extentX=11.1875, extentY=11.3125, extentZ=1.3125, center=<5.2189 -4.2 0>}]
 dyn_skirt_fr_l_03 [{shapeType=Box, extentX=11.1875, extentY=11.8125, extentZ=1.3125, center=<5.2189 -4.2 0.0002>}]
 dyn_skirt_fr_mid_l_01 [{shapeType=Box, extentX=11.1875, extentY=16.2935, extentZ=1.3125, center=<5.2189 0 0>}]
 dyn_skirt_fr_mid_l_02 [{shapeType=Box, extentX=11.1875, extentY=15.3125, extentZ=1.3125, center=<5.2189 0 0>}]
 dyn_skirt_fr_mid_l_03 [{shapeType=Box, extentX=11.1875, extentY=15.3125, extentZ=1.3125, center=<5.2189 0 0>}]
 dyn_skirt_fr_mid_l_MSTR [{shapeType=Capsule, length=1, radius=1}]
 dyn_skirt_fr_mid_r_01 [{shapeType=Box, extentX=11.1875, extentY=16.2935, extentZ=1.3125, center=<5.2189 0 0>}]
 dyn_skirt_fr_mid_r_02 [{shapeType=Box, extentX=11.1875, extentY=24.5, extentZ=1.3125, center=<5.219 3.3 -1.3>, rotation=<rot 0 0 21.0035>}]
 dyn_skirt_fr_mid_r_03 [{shapeType=Box, extentX=11.1875, extentY=24.5, extentZ=1.3125, center=<5.219 3.3 -1.3>, rotation=<rot 0 0 9.4031>}]
 dyn_skirt_fr_mid_r_MSTR [{shapeType=Capsule, length=1, radius=1}]
 dyn_skirt_fr_r_01 [{shapeType=Box, extentX=11.1875, extentY=16.2935, extentZ=1.3125, center=<5.2189 4.2 0>}]
 dyn_skirt_fr_r_02 [{shapeType=Box, extentX=11.1875, extentY=18.875, extentZ=1.3125, center=<5.2189 8 -0.0001>, rotation=<rot 0 0 14.187>}]
 dyn_skirt_fr_r_03 [{shapeType=Box, extentX=11.1875, extentY=18.875, extentZ=1.3125, center=<5.2189 8 0>, rotation=<rot 0 0 14.187>}]
 dyn_skirt_l_01 [{shapeType=Box, extentX=11.1875, extentY=16.2935, extentZ=1.3125, center=<5.2189 0 0>}]
 dyn_skirt_l_02 [{shapeType=Box, extentX=11.1875, extentY=24.8791, extentZ=1.3125, center=<5.2189 0 0>}]
 dyn_skirt_l_03 [{shapeType=Box, extentX=11.1875, extentY=24.8791, extentZ=1.3125, center=<5.2189 0 0>}]
 dyn_skirt_l_MSTR [{shapeType=Capsule, length=1, radius=1}]
 dyn_skirt_r_01 [{shapeType=Box, extentX=11.1875, extentY=16.2935, extentZ=1.3125, center=<5.2189 0 0>}]
 dyn_skirt_r_02 [{shapeType=Box, extentX=11.1875, extentY=24.8791, extentZ=1.3125, center=<5.2189 0 0>}]
 dyn_skirt_r_03 [{shapeType=Box, extentX=11.1875, extentY=24.8791, extentZ=1.3125, center=<5.2189 0 0>}]
 dyn_skirt_r_MSTR [{shapeType=Capsule, length=1, radius=1}]
 pelvis [{shapeType=Capsule, length=8.125, radius=11.5312, center=<-8.3294 0 -0.0007>}]
 spine_01 [{shapeType=Capsule, length=9.0938, radius=10.4375, center=<-7.086 0.3334 -0.0007>}]
 thigh_l [{shapeType=Capsule, length=24.8125, radius=11.5625, center=<-19.1301 -0.6906 2.0566>, rotation=<rot 84.762 -179.9999 179.2646>}]
 thigh_r [{shapeType=Capsule, length=31.3438, radius=10.1562, center=<18.9211 -0.2689 0.1764>, rotation=<rot 84.7658 -179.9999 177.495>}]
}
constraints [
 {bone1Name=dyn_skirt_bk_l_02, bone2Name=dyn_skirt_bk_l_01, swing1LimitDegrees=2, swing2LimitDegrees=10, twistLimitDegrees=5}
 {bone1Name=dyn_skirt_bk_l_03, bone2Name=dyn_skirt_bk_l_02, swing1LimitDegrees=2, swing2LimitDegrees=10, twistLimitDegrees=5}
 {bone1Name=dyn_skirt_bk_02, bone2Name=dyn_skirt_bk_01, swing1LimitDegrees=2, swing2LimitDegrees=10, twistLimitDegrees=5}
 {bone1Name=dyn_skirt_bk_03, bone2Name=dyn_skirt_bk_02, swing1LimitDegrees=2, swing2LimitDegrees=10, twistLimitDegrees=5}
 {bone1Name=dyn_skirt_fr_mid_l_01, bone2Name=dyn_skirt_fr_mid_l_MSTR, swing1LimitDegrees=2, swing2LimitDegrees=45, twistLimitDegrees=5}
 {bone1Name=dyn_skirt_fr_mid_l_02, bone2Name=dyn_skirt_fr_mid_l_01, swing1LimitDegrees=2, swing2LimitDegrees=45, twistLimitDegrees=5}
 {bone1Name=dyn_skirt_fr_mid_l_03, bone2Name=dyn_skirt_fr_mid_l_02, swing1LimitDegrees=2, swing2LimitDegrees=45, twistLimitDegrees=5}
 {bone1Name=dyn_skirt_fr_l_01, bone2Name=dyn_skirt_fr_MSTR, swing1LimitDegrees=45, swing2LimitDegrees=45, twistLimitDegrees=5}
 {bone1Name=dyn_skirt_fr_l_02, bone2Name=dyn_skirt_fr_l_01, swing1LimitDegrees=45, swing2LimitDegrees=45, twistLimitDegrees=5}
 {bone1Name=dyn_skirt_fr_l_03, bone2Name=dyn_skirt_fr_l_02, swing1LimitDegrees=45, swing2LimitDegrees=45, twistLimitDegrees=5}
 {bone1Name=dyn_skirt_fr_r_01, bone2Name=dyn_skirt_fr_MSTR, swing1LimitDegrees=45, swing2LimitDegrees=45, twistLimitDegrees=5}
 {bone1Name=dyn_skirt_fr_r_02, bone2Name=dyn_skirt_fr_r_01, swing1LimitDegrees=45, swing2LimitDegrees=45, twistLimitDegrees=5}
 {bone1Name=dyn_skirt_fr_r_03, bone2Name=dyn_skirt_fr_r_02, swing1LimitDegrees=45, swing2LimitDegrees=45, twistLimitDegrees=5}
 {bone1Name=dyn_skirt_l_01, bone2Name=dyn_skirt_l_MSTR, swing1LimitDegrees=5, swing2LimitDegrees=45, twistLimitDegrees=5}
 {bone1Name=dyn_skirt_l_02, bone2Name=dyn_skirt_l_01, swing1LimitDegrees=5, swing2LimitDegrees=45, twistLimitDegrees=5}
 {bone1Name=dyn_skirt_l_03, bone2Name=dyn_skirt_l_02, swing1LimitDegrees=5, swing2LimitDegrees=45, twistLimitDegrees=5}
 {bone1Name=dyn_skirt_fr_mid_r_01, bone2Name=dyn_skirt_fr_mid_r_MSTR, swing1LimitDegrees=3, swing2LimitDegrees=45, twistLimitDegrees=5}
 {bone1Name=dyn_skirt_fr_mid_r_02, bone2Name=dyn_skirt_fr_mid_r_01, swing1LimitDegrees=3, swing2LimitDegrees=45, twistLimitDegrees=5}
 {bone1Name=dyn_skirt_fr_mid_r_03, bone2Name=dyn_skirt_fr_mid_r_02, swing1LimitDegrees=3, swing2LimitDegrees=45, twistLimitDegrees=5}
 {bone1Name=dyn_skirt_r_01, bone2Name=dyn_skirt_r_MSTR, swing1LimitDegrees=5, swing2LimitDegrees=45, twistLimitDegrees=5}
 {bone1Name=dyn_skirt_r_02, bone2Name=dyn_skirt_r_01, swing1LimitDegrees=5, swing2LimitDegrees=45, twistLimitDegrees=5}
 {bone1Name=dyn_skirt_r_03, bone2Name=dyn_skirt_r_02, swing1LimitDegrees=5, swing2LimitDegrees=45, twistLimitDegrees=5}
 {bone1Name=dyn_skirt_bk_r_02, bone2Name=dyn_skirt_bk_r_01, swing1LimitDegrees=2, swing2LimitDegrees=10, twistLimitDegrees=5}
 {bone1Name=dyn_skirt_bk_r_03, bone2Name=dyn_skirt_bk_r_02, swing1LimitDegrees=2, swing2LimitDegrees=10, twistLimitDegrees=5}
 {bone1Name=dyn_skirt_bk_l_01, bone2Name=pelvis, swing1LimitDegrees=2, swing2LimitDegrees=10, twistLimitDegrees=5}
 {bone1Name=dyn_skirt_bk_01, bone2Name=pelvis, swing1LimitDegrees=2, swing2LimitDegrees=10, twistLimitDegrees=5}
 {bone1Name=dyn_skirt_bk_r_01, bone2Name=pelvis, swing1LimitDegrees=2, swing2LimitDegrees=10, twistLimitDegrees=5}
 {
  bone1Name=dyn_skirt_fr_l_03
  bone2Name=dyn_skirt_fr_mid_l_03
  swing1LimitDegrees=45
  swing1Motion=Free
  swing2LimitDegrees=45
  swing2Motion=Free
  twistLimitDegrees=45
  twistMotion=Free
 }
 {
  bone1Name=dyn_skirt_fr_r_03
  bone2Name=dyn_skirt_fr_mid_r_03
  swing1LimitDegrees=45
  swing1Motion=Free
  swing2LimitDegrees=45
  swing2Motion=Free
  twistLimitDegrees=45
  twistMotion=Free
 }
 {
  bone1Name=dyn_skirt_l_03
  bone2Name=dyn_skirt_fr_mid_l_03
  swing1LimitDegrees=45
  swing1Motion=Free
  swing2LimitDegrees=45
  swing2Motion=Free
  twistLimitDegrees=45
  twistMotion=Free
 }
 {
  bone1Name=dyn_skirt_r_03
  bone2Name=dyn_skirt_fr_mid_r_03
  swing1LimitDegrees=45
  swing1Motion=Free
  swing2LimitDegrees=45
  swing2Motion=Free
  twistLimitDegrees=45
  twistMotion=Free
 }
 {
  bone1Name=dyn_skirt_r_02
  bone2Name=dyn_skirt_fr_mid_r_02
  swing1LimitDegrees=45
  swing1Motion=Free
  swing2LimitDegrees=45
  swing2Motion=Free
  twistLimitDegrees=45
  twistMotion=Free
 }
 {
  bone1Name=dyn_skirt_bk_r_02
  bone2Name=dyn_skirt_r_02
  swing1LimitDegrees=45
  swing1Motion=Free
  swing2LimitDegrees=45
  swing2Motion=Free
  twistLimitDegrees=45
  twistMotion=Free
 }
 {
  bone1Name=dyn_skirt_bk_l_02
  bone2Name=dyn_skirt_bk_02
  swing1LimitDegrees=45
  swing1Motion=Free
  swing2LimitDegrees=45
  swing2Motion=Free
  twistLimitDegrees=45
  twistMotion=Free
 }
 {
  bone1Name=dyn_skirt_r_02
  bone2Name=dyn_skirt_bk_r_02
  swing1LimitDegrees=45
  swing1Motion=Free
  swing2LimitDegrees=45
  swing2Motion=Free
  twistLimitDegrees=45
  twistMotion=Free
 }
 {
  bone1Name=dyn_skirt_l_02
  bone2Name=dyn_skirt_bk_l_02
  swing1LimitDegrees=45
  swing1Motion=Free
  swing2LimitDegrees=45
  swing2Motion=Free
  twistLimitDegrees=45
  twistMotion=Free
 }
]
