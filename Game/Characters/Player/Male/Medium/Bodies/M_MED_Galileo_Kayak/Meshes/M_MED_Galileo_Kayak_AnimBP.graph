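# Anim Blueprint for the Galileo Kayak male player body, copying the attached parent mesh's pose and adding secondary motion to its skirt bones.
# A cached pose with crouch and thigh-driven skirt offsets feeds Control Rig and RigidBody branches, selected by skydiving and cape state.
# Thigh constraints, bone offsets and a head scale shape the result.
# The event graph updates those inputs and switches physics gravity by vertical velocity during freefall.
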AnimBlueprint /Game/Characters/Player/Male/Medium/Bodies/M_MED_Galileo_Kayak/Meshes/M_MED_Galileo_Kayak_AnimBP
target_skeleton=&/M_MED_Galileo_Kayak_Skeleton
graphs [
 AnimGraph {nodes [
   Blend Poses by bool(bActiveValue=Get bIsMoving2D()) {blendType=Linear}
   n1 = Control Rig(Source=Use cached pose 'initial Pose'()) {
    alphaBoolBlend {blendOption=Linear}
    controlRigAssetReference {blueprintRigClass=&/M_MED_Galileo_Kayak_CtrlRig.M_MED_Galileo_Kayak_CtrlRig_C}
   }
   n2 = Get gravity_override() -> (gravity_override:Vector)
   n3 = RigidBody!(ComponentPose=Local To Component(LocalPose=@n1), Alpha, OverrideWorldGravity:Vector=@n2, ExternalForce) -> (Pose) {
    bClampLinearTranslationLimitToRefPose=true
    bOverrideWorldGravity=true
    overridePhysicsAsset=&/M_MED_Galileo_Kayak_Dyn_new
    simulationSpace=BaseBoneSpace
    alphaBoolBlend {blendOption=Linear}
    baseBoneRef {boneName=spine_01}
    componentAppliedLinearAccClamp=<3000 3000 3000>
    componentLinearAccScale=<0.3 0.2 0.2>
    componentLinearVelScale=<0.3 0.2 0.2>
    overrideWorldGravity {z=-980}
   }
   n4 = "Constraint - dyn_skirt_fr_l_01 (P)"(ComponentPose=@n3, Alpha=0.2, ConstraintWeights_0=1) {
    alphaBoolBlend {blendOption=Linear}
    boneToModify {boneName=dyn_skirt_fr_l_01}
    constraintSetup [{targetBone {boneName=thigh_l}}]
    constraintWeights [1]
   }
   n5 = "Constraint - dyn_skirt_fr_r_01 (P)"(ComponentPose=@n4, Alpha=0.2, ConstraintWeights_0=1) {
    alphaBoolBlend {blendOption=Linear}
    boneToModify {boneName=dyn_skirt_fr_r_01}
    constraintSetup [{targetBone {boneName=thigh_r}}]
    constraintWeights [1]
   }
   n6 = "Constraint - dyn_skirt_fr_l_02 (P)"(ComponentPose=@n5, Alpha=0.2, ConstraintWeights_0=1) {
    alphaBoolBlend {blendOption=Linear}
    boneToModify {boneName=dyn_skirt_fr_l_02}
    constraintSetup [{targetBone {boneName=thigh_l}}]
    constraintWeights [1]
   }
   n7 = "Constraint - dyn_skirt_fr_r_02 (P)"(ComponentPose=@n6, Alpha=0.2, ConstraintWeights_0=1) {
    alphaBoolBlend {blendOption=Linear}
    boneToModify {boneName=dyn_skirt_fr_r_02}
    constraintSetup [{targetBone {boneName=thigh_r}}]
    constraintWeights [1]
   }
   n8 = "Constraint - dyn_skirt_fr_l_03 (P)"(ComponentPose=@n7, Alpha=0.2, ConstraintWeights_0=1) {
    alphaBoolBlend {blendOption=Linear}
    boneToModify {boneName=dyn_skirt_fr_l_03}
    constraintSetup [{targetBone {boneName=thigh_l}}]
    constraintWeights [1]
   }
   n9 = "Constraint - dyn_skirt_fr_r_03 (P)"(ComponentPose=@n8, Alpha=0.2, ConstraintWeights_0=1) {
    alphaBoolBlend {blendOption=Linear}
    boneToModify {boneName=dyn_skirt_fr_r_03}
    constraintSetup [{targetBone {boneName=thigh_r}}]
    constraintWeights [1]
   }
   Component To Local(ComponentPose=@n9)
   "Transform (Modify) Bone - Bone: dyn_skirt_fr_mid_r_MSTR"(Alpha=Get jumping alpha() -> (jumping alpha:double), Translation="0,0,0", Rotation="90,0,0", Scale="1,1,1") {
    rotationMode=BMM_Additive
    rotationSpace=BCS_BoneSpace
    translationMode=BMM_Additive
    translationSpace=BCS_BoneSpace
    alphaBoolBlend {blendOption=Linear}
    alphaScaleBiasClamp {bInterpResult=true}
    boneToModify {boneName=dyn_skirt_fr_mid_r_MSTR}
    rotation {pitch=90}
   }
   n10 = Control Rig(Source=Use cached pose 'initial Pose'()) {
    alphaBoolBlend {blendOption=Linear}
    controlRigAssetReference {blueprintRigClass=&M_MED_Galileo_Kayak_CtrlRig_C}
   }
   n11 = RigidBody(ComponentPose=Local To Component(LocalPose=@n10), OverrideWorldGravity=Get gravity_override()) {
    bClampLinearTranslationLimitToRefPose=true
    bOverrideWorldGravity=true
    overridePhysicsAsset=&/M_MED_Galileo_Kayak_Dyn_skydiving
    simulationSpace=BaseBoneSpace
    alphaBoolBlend {blendOption=Linear}
    baseBoneRef {boneName=spine_01}
    componentAppliedLinearAccClamp=<3000 3000 3000>
    componentLinearAccScale=<0.3 0.2 0.2>
    componentLinearVelScale=<0.3 0.2 0.2>
    overrideWorldGravity {z=-980}
   }
   n12 = "Constraint - dyn_skirt_l_MSTR (P)"!(ComponentPose=@n11, bAlphaBoolEnabled:bool=True, ConstraintWeights_0=1) -> (Pose) {
    alphaInputType=Bool
    alphaBoolBlend {blendOption=Linear}
    boneToModify {boneName=dyn_skirt_l_MSTR}
    constraintSetup [{targetBone {boneName=thigh_twist_01_l}}]
    constraintWeights [1]
   }
   n13 = "Constraint - dyn_skirt_r_MSTR (P)"!(ComponentPose=@n12, bAlphaBoolEnabled:bool=True, ConstraintWeights_0=1) -> (Pose) {
    alphaInputType=Bool
    alphaBoolBlend {blendOption=Linear}
    boneToModify {boneName=dyn_skirt_r_MSTR}
    constraintSetup [{targetBone {boneName=thigh_twist_01_r}}]
    constraintWeights [1]
   }
   n14 = "Constraint - dyn_skirt_bk_l_MSTR (P)"!(ComponentPose=@n13, bAlphaBoolEnabled:bool=True, ConstraintWeights_0=1) -> (Pose) {
    alphaInputType=Bool
    alphaBoolBlend {blendOption=Linear}
    boneToModify {boneName=dyn_skirt_bk_l_MSTR}
    constraintSetup [{targetBone {boneName=thigh_twist_01_l}}]
    constraintWeights [1]
   }
   n15 = "Constraint - dyn_skirt_bk_r_MSTR (P)"!(ComponentPose=@n14, bAlphaBoolEnabled:bool=True, ConstraintWeights_0=1) -> (Pose) {
    alphaInputType=Bool
    alphaBoolBlend {blendOption=Linear}
    boneToModify {boneName=dyn_skirt_bk_r_MSTR}
    constraintSetup [{targetBone {boneName=thigh_twist_01_r}}]
    constraintWeights [1]
   }
   n16 = "Transform (Modify) Bone - Bone: dyn_skirt_l_MSTR"(ComponentPose=@n15, Alpha=1, Translation="0,0,2", Rotation="15,0,0", Scale="1,1,1") {
    rotationMode=BMM_Additive
    rotationSpace=BCS_BoneSpace
    translationMode=BMM_Additive
    translationSpace=BCS_BoneSpace
    alphaBoolBlend {blendOption=Linear}
    boneToModify {boneName=dyn_skirt_l_MSTR}
    rotation {pitch=15}
    translation {z=2}
   }
   n17 = "Transform (Modify) Bone - Bone: dyn_skirt_r_MSTR"(ComponentPose=@n16, Alpha=1, Translation="0,0,2", Rotation="15,0,0", Scale="1,1,1") {
    rotationMode=BMM_Additive
    rotationSpace=BCS_BoneSpace
    translationMode=BMM_Additive
    translationSpace=BCS_BoneSpace
    alphaBoolBlend {blendOption=Linear}
    boneToModify {boneName=dyn_skirt_r_MSTR}
    rotation {pitch=15}
    translation {z=2}
   }
   n18 = "Transform (Modify) Bone - Bone: dyn_skirt_bk_l_MSTR"(ComponentPose=@n17, Alpha=1, Translation="0,0,2", Rotation="15,0,0", Scale="1,1,1") {
    rotationMode=BMM_Additive
    rotationSpace=BCS_BoneSpace
    translationMode=BMM_Additive
    translationSpace=BCS_BoneSpace
    alphaBoolBlend {blendOption=Linear}
    boneToModify {boneName=dyn_skirt_bk_l_MSTR}
    rotation {pitch=15}
    translation {z=2}
   }
   n19 = "Transform (Modify) Bone - Bone: dyn_skirt_bk_r_MSTR"(ComponentPose=@n18, Alpha=1, Translation="0,0,2", Rotation="15,0,0", Scale="1,1,1") {
    rotationMode=BMM_Additive
    rotationSpace=BCS_BoneSpace
    translationMode=BMM_Additive
    translationSpace=BCS_BoneSpace
    alphaBoolBlend {blendOption=Linear}
    boneToModify {boneName=dyn_skirt_bk_r_MSTR}
    rotation {pitch=15}
    translation {z=2}
   }
   n20 = "Constraint - dyn_skirt_bk_MSTR (PP)"!(ComponentPose=@n19, Alpha=1, ConstraintWeights_0=0.5, ConstraintWeights_1:float=0.5) -> (Pose) {
    alphaBoolBlend {blendOption=Linear}
    boneToModify {boneName=dyn_skirt_bk_MSTR}
    constraintSetup [{targetBone {boneName=dyn_skirt_bk_r_MSTR}}, {targetBone {boneName=dyn_skirt_bk_l_MSTR}}]
    constraintWeights [1,1]
   }
   n21 = Component To Local(ComponentPose=@n20)
   n22 = Control Rig(Source=Use cached pose 'initial Pose'()) {
    alphaBoolBlend {blendOption=Linear}
    controlRigAssetReference {blueprintRigClass=&M_MED_Galileo_Kayak_CtrlRig_C}
   }
   n23 = RigidBody(ComponentPose=Local To Component(LocalPose=@n22), OverrideWorldGravity=@n2) {
    bClampLinearTranslationLimitToRefPose=true
    bOverrideWorldGravity=true
    overridePhysicsAsset=&/M_MED_Galileo_Kayak_Dyn_withCape
    simulationSpace=BaseBoneSpace
    alphaBoolBlend {blendOption=Linear}
    baseBoneRef {boneName=spine_01}
    componentAppliedLinearAccClamp=<3000 3000 3000>
    componentLinearAccScale=<0.3 0.2 0.2>
    componentLinearVelScale=<0.3 0.2 0.2>
    overrideWorldGravity {z=-980}
   }
   n24 = Component To Local(ComponentPose=@n23)
   n25 = Control Rig!(Source=Use cached pose 'initial Pose'(), Alpha, RBAN off:bool=NOT Boolean(A=Get bRigidBodyEnabled())) -> (Pose) {
    alphaBoolBlend {blendOption=Linear}
    controlRigAssetReference {blueprintRigClass=&M_MED_Galileo_Kayak_CtrlRig_C}
   }
   n26 = RigidBody(ComponentPose=Local To Component(LocalPose=@n25), OverrideWorldGravity=@n2) {
    bClampLinearTranslationLimitToRefPose=true
    bOverrideWorldGravity=true
    overridePhysicsAsset=&/M_MED_Galileo_Kayak_Dyn
    simulationSpace=BaseBoneSpace
    alphaBoolBlend {blendOption=Linear}
    baseBoneRef {boneName=spine_01}
    componentAppliedLinearAccClamp=<3000 3000 3000>
    componentLinearAccScale=<0.3 0.2 0.2>
    componentLinearVelScale=<0.3 0.2 0.2>
    overrideWorldGravity {z=-980}
   }
   n27 = Component To Local(ComponentPose=@n26)
   n28 = Get isWearingCape() -> (isWearingCape:bool)
   n29 = NOT Boolean(A=Get bRigidBodyEnabled())
   n30 = "Constraint - dyn_skirt_fr_mid_r_MSTR (P)"!(
    ComponentPose=Local To Component(LocalPose=Blend Poses by bool(BlendPose_0=@n24, BlendPose_1=@n27, bActiveValue=@n28) {blendType=Linear})
    bAlphaBoolEnabled:bool=@n29
    ConstraintWeights_0=1
   ) -> (Pose) {
    alphaInputType=Bool
    alphaBoolBlend {blendInTime=0.3, blendOption=ExpInOut, blendOutTime=0.3}
    boneToModify {boneName=dyn_skirt_fr_mid_r_MSTR}
    constraintSetup [{targetBone {boneName=thigh_r}}]
    constraintWeights [1]
   }
   n31 = "Constraint - dyn_skirt_fr_l_01 (PP)"!(ComponentPose=@n30, bAlphaBoolEnabled:bool=@n29, ConstraintWeights_0=0.9, ConstraintWeights_1:float=0.1) -> (Pose) {
    alphaInputType=Bool
    alphaBoolBlend {blendInTime=0.3, blendOption=ExpInOut, blendOutTime=0.3}
    boneToModify {boneName=dyn_skirt_fr_l_01}
    constraintSetup [{targetBone {boneName=thigh_l}}, {targetBone {boneName=pelvis}}]
    constraintWeights [1,1]
   }
   n32 = "Constraint - dyn_skirt_fr_mid_l_MSTR (P)"!(ComponentPose=@n31, bAlphaBoolEnabled:bool=@n29, ConstraintWeights_0=1) -> (Pose) {
    alphaInputType=Bool
    alphaBoolBlend {blendInTime=0.3, blendOption=ExpInOut, blendOutTime=0.3}
    boneToModify {boneName=dyn_skirt_fr_mid_l_MSTR}
    constraintSetup [{targetBone {boneName=thigh_l}}]
    constraintWeights [1]
   }
   n33 = "Constraint - dyn_skirt_fr_r_01 (PP)"!(ComponentPose=@n32, bAlphaBoolEnabled:bool=@n29, ConstraintWeights_0=0.9, ConstraintWeights_1:float=0.1) -> (Pose) {
    alphaInputType=Bool
    alphaBoolBlend {blendInTime=0.3, blendOption=ExpInOut, blendOutTime=0.3}
    boneToModify {boneName=dyn_skirt_fr_r_01}
    constraintSetup [{targetBone {boneName=thigh_r}}, {targetBone {boneName=pelvis}}]
    constraintWeights [1,1]
   }
   n34 = "Transform (Modify) Bone - Bone: dyn_skirt_fr_mid_l_MSTR"!(ComponentPose=@n33, bAlphaBoolEnabled:bool=@n29, Translation="0,0,0", Rotation="0,0,0", Scale="1,1,1") -> (Pose) {
    alphaInputType=Bool
    rotationMode=BMM_Additive
    rotationSpace=BCS_BoneSpace
    alphaBoolBlend {blendInTime=0.3, blendOption=ExpInOut, blendOutTime=0.3}
    boneToModify {boneName=dyn_skirt_fr_mid_l_MSTR}
   }
   n35 = Component To Local(ComponentPose=@n34)
   n36 = Get bIsSkyDiving()
   n37 = "Constraint - dyn_skirt_l_MSTR (PP)"!(
    ComponentPose=Local To Component(LocalPose=Blend Poses by bool(BlendPose_0=@n21, BlendPose_1=@n35, bActiveValue=@n36) {blendType=Linear})
    bAlphaBoolEnabled:bool=True
    ConstraintWeights_0=1
    ConstraintWeights_1:float=0
   ) -> (Pose) {
    alphaInputType=Bool
    alphaBoolBlend {blendOption=Linear}
    boneToModify {boneName=dyn_skirt_l_MSTR}
    constraintSetup [{targetBone {boneName=thigh_twist_01_l}}, {targetBone {boneName=dyn_skirt_fr_mid_l_01}}]
    constraintWeights [1,1]
   }
   n38 = "Constraint - dyn_skirt_r_MSTR (PP)"!(ComponentPose=@n37, bAlphaBoolEnabled:bool=True, ConstraintWeights_0=1, ConstraintWeights_1:float=0) -> (Pose) {
    alphaInputType=Bool
    alphaBoolBlend {blendOption=Linear}
    boneToModify {boneName=dyn_skirt_r_MSTR}
    constraintSetup [{targetBone {boneName=thigh_twist_01_r}}, {targetBone {boneName=dyn_skirt_fr_mid_r_01}}]
    constraintWeights [1,1]
   }
   n39 = "Transform (Modify) Bone - Bone: head"!(ComponentPose=@n38, Alpha=1) -> (Pose) {scaleMode=BMM_Replace, alphaBoolBlend {blendOption=Linear}, boneToModify {boneName=head}, scale=<0.93 0.93 0.93>}
   n40 = "Transform (Modify) Bone - Bone: dyn_skirt_bk_l_01"(ComponentPose=@n39, Alpha=1, Translation="0,0,0.5", Rotation="0,0,0", Scale="1,1,1") {
    rotationMode=BMM_Additive
    rotationSpace=BCS_BoneSpace
    translationMode=BMM_Additive
    translationSpace=BCS_BoneSpace
    alphaBoolBlend {blendOption=Linear}
    boneToModify {boneName=dyn_skirt_bk_l_01}
    translation {z=0.5}
   }
   n41 = "Transform (Modify) Bone - Bone: dyn_skirt_bk_01"(ComponentPose=@n40, Alpha=1, Translation="0,0,0.5", Rotation="0,0,0", Scale="1,1,1") {
    rotationMode=BMM_Additive
    rotationSpace=BCS_BoneSpace
    translationMode=BMM_Additive
    translationSpace=BCS_BoneSpace
    alphaBoolBlend {blendOption=Linear}
    boneToModify {boneName=dyn_skirt_bk_01}
    translation {z=0.5}
   }
   n42 = "Transform (Modify) Bone - Bone: dyn_skirt_bk_r_01"(ComponentPose=@n41, Alpha=1, Translation="0,0,0.5", Rotation="0,0,0", Scale="1,1,1") {
    rotationMode=BMM_Additive
    rotationSpace=BCS_BoneSpace
    translationMode=BMM_Additive
    translationSpace=BCS_BoneSpace
    alphaBoolBlend {blendOption=Linear}
    boneToModify {boneName=dyn_skirt_bk_r_01}
    translation {z=0.5}
   }
   Output Pose(Result=Component To Local(ComponentPose=@n42)) {layerGroup=None}
   n43 = Get crouch alpha() -> (crouch alpha:double)
   n44 = "Transform (Modify) Bone - Bone: dyn_skirt_fr_mid_l_MSTR"!(
    ComponentPose=Local To Component(LocalPose=Copy Pose From Mesh() {bUseAttachedParent=true})
    Alpha:float=@n43
    Translation="-7.731,-1.109,3.397"
    Rotation="10,0,-9.399"
    Scale="1,1,1"
   ) -> (Pose) {
    rotationMode=BMM_Additive
    rotationSpace=BCS_BoneSpace
    translationMode=BMM_Additive
    translationSpace=BCS_BoneSpace
    alphaBoolBlend {blendOption=Linear}
    alphaScaleBiasClamp {bInterpResult=true, interpSpeedDecreasing=20, interpSpeedIncreasing=20}
    boneToModify {boneName=dyn_skirt_fr_mid_l_MSTR}
    rotation {pitch=10, roll=-9.399}
    translation=<-7.731 -1.109 3.397>
   }
   n45 = "Transform (Modify) Bone - Bone: dyn_skirt_fr_MSTR"(ComponentPose=@n44, Alpha=@n43, Translation="-7.285,-1.330,2.805", Rotation="0,0,0", Scale="1,1,1") {
    rotationMode=BMM_Additive
    rotationSpace=BCS_BoneSpace
    translationMode=BMM_Additive
    translationSpace=BCS_BoneSpace
    alphaBoolBlend {blendOption=Linear}
    alphaScaleBiasClamp {bInterpResult=true, interpSpeedDecreasing=20, interpSpeedIncreasing=20}
    boneToModify {boneName=dyn_skirt_fr_MSTR}
    translation=<-7.285 -1.33 2.805>
   }
   n46 = "Transform (Modify) Bone - Bone: dyn_skirt_fr_mid_r_MSTR"(
    ComponentPose=@n45
    Alpha=Get thigh rotation_r() -> (thigh rotation_r:double)
    Translation="-6.552,0.000,0.000"
    Rotation="0,0,0"
    Scale="1,1,1"
   ) {
    rotationMode=BMM_Additive
    rotationSpace=BCS_BoneSpace
    translationMode=BMM_Additive
    translationSpace=BCS_BoneSpace
    alphaBoolBlend {blendOption=Linear}
    alphaScaleBiasClamp {bInterpResult=true}
    boneToModify {boneName=dyn_skirt_fr_mid_r_MSTR}
    translation {x=-6.552}
   }
   Save cached pose 'initial Pose'(Pose=Component To Local(ComponentPose=@n46)) {cacheName=initial Pose}
  ]}
 EventGraph {nodes [
   Event BlueprintUpdateAnimation()
   > n1 = Sequence!(execute) -> (then_0, then_1, then_2:Exec)
   > n2 = Branch(Condition=Get bIsSkyDiving())
   > n3 = Branch(Condition=Get bIsParachuteOpened())
   Set gravity_override(execute=[@n2.else, @n3.then], gravity_override="0, 0,-980")
   n4 = Branch(execute=@n3.else, Condition="float > float"(A=Get PawnVelocityZ(), B=50) {is=title K2Node_CallFunction})
   > Set gravity_override(gravity_override="0, 0,200")
   Set gravity_override(execute=@n4.else, gravity_override="0, 0,-200")
   n5 = BreakTransform(InTransform=GetDeltaTransformFromRefPose(self=GetOwnerSkeletalMeshComponent(), BoneName=thigh_r, BaseName=pelvis))
   Set thigh rotation_r(execute=@n1.then_1, thigh rotation_r:double=MapRangeClamped(Value=BreakRotator(InRot=@n5.Rotation).Yaw, InRangeA=-50, InRangeB=-90)) -> (Output_Get:double)
   n6 = BreakTransform(InTransform=GetDeltaTransformFromRefPose(self=GetOwnerSkeletalMeshComponent(), BoneName=thigh_l, BaseName=pelvis))
   > Set thigh_rotation_l(thigh_rotation_l:double=MapRangeClamped(Value=BreakRotator(InRot=@n6.Rotation).Yaw, InRangeA=-50, InRangeB=-90)) -> (Output_Get:double)
   > n7 = Branch(Condition=Get bIsJumping())
   Set jumping alpha(execute=[@n7.then, @n8.then], jumping alpha:double=1) -> (Output_Get:double)
   n8 = Branch(execute=@n7.else, Condition=Get bIsFalling())
   Set jumping alpha(execute=@n8.else, jumping alpha=0)
   Event OnBeginCrouching()
   > Set crouch alpha(crouch alpha:double=1) -> (Output_Get:double)
   Event OnEndCrouching()
   > Set crouch alpha(crouch alpha=0)
   Event OnCharacterCustomizationLoaded()
   n9 = CheckForBackbling()
   > Set isWearingBackbling(isWearingBackbling=@n9)
   > Set isWearingCape(isWearingCape=@n9.bIsBackblingLong)
  ]}
]
variables [
 crouch alpha {type=number}
 gravity_override {type=Vector}
 Idle {type=boolean}
 isWearingBackbling {type=boolean}
 isWearingCape {type=boolean}
 jumping alpha {type=number}
 thigh rotation_r {type=number}
 thigh_rotation_l {type=number}
]
